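# Physics asset for the F_MED_BudBlossom female medium body's choker, defining bodies and joints for its dangling chain and collar bones.
# The spine_05 body uses a capsule plus an angled box as the torso volume, with chain and collar base bodies joined to it by free-rotation constraints.
# Both chain segments constrain directly to the chain base rather than in series, and the collar segment to its base, all with angular limits.
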PhysicsAsset /BRCosmetics/Characters/Player/Female/Medium/Bodies/F_MED_BudBlossom/Meshes/F_MED_BudBlossom_Choker_Physics
bodies {
 dyn_chain_01 [{shapeType=Capsule, length=1.2412, radius=1.2, shapeName=dyn_chain_01_capsule, center=<2.0066 0 0>}]
 dyn_chain_02 [{shapeType=Capsule, length=1.8466, radius=1.4, shapeName=dyn_chain_02_capsule, center=<3.9278 -0.0936 0>}]
 dyn_chain_base [{shapeType=Capsule, length=0.505, radius=0.505, shapeName=dyn_chain_base_capsule}]
 dyn_collar_01 [{shapeType=Capsule, length=6.6553, radius=1.3452, shapeName=dyn_collar_01_capsule, center=<3.4269 0 0.1354>, rotation=<rot 0 0 -90>}]
 dyn_collar_base [{shapeType=Capsule, length=0.505, radius=0.505, shapeName=dyn_collar_base_capsule, rotation=<rot 0 -90 0>}]
 spine_05 [
  {shapeType=Capsule, length=9.9331, radius=7.9093, shapeName=spine_05_capsule}
  {shapeType=Box, extentX=7.5, extentY=4.5, extentZ=7.5, shapeName=spine_05_box, center=<8.2396 -1.8316 0>, rotation=<rot 0 42.8988 0>}
 ]
}
constraints [
 {bone1Name=dyn_chain_01, bone2Name=dyn_chain_base, swing1LimitDegrees=45, swing2LimitDegrees=5, twistLimitDegrees=10}
 {bone1Name=dyn_chain_02, bone2Name=dyn_chain_base, swing1LimitDegrees=40, swing2LimitDegrees=5, twistLimitDegrees=10}
 {
  bone1Name=dyn_chain_base
  bone2Name=spine_05
  swing1LimitDegrees=45
  swing1Motion=Free
  swing2LimitDegrees=45
  swing2Motion=Free
  twistLimitDegrees=45
  twistMotion=Free
 }
 {
  bone1Name=dyn_collar_base
  bone2Name=spine_05
  swing1LimitDegrees=45
  swing1Motion=Free
  swing2LimitDegrees=45
  swing2Motion=Free
  twistLimitDegrees=45
  twistMotion=Free
 }
 {bone1Name=dyn_collar_01, bone2Name=dyn_collar_base, swing1LimitDegrees=10, swing2LimitDegrees=30, twistLimitDegrees=5}
]
