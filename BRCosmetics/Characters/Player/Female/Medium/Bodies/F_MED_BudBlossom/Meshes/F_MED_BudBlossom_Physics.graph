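# Physics asset for the BudBlossom medium female player body cosmetic, covering dangling whip and back-skirt bones.
# Capsules cover pelvis, both thighs and only the right calf and foot, while a large root sphere sits mostly below the origin, resembling a floor.
# Both chains branch from pelvis through tiny base capsules, with the whip forking into segment and handle.
# Chain joints past the bases have tight angular limits, twist-locked on the whip; other constraints swing and twist freely.
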PhysicsAsset /BRCosmetics/Characters/Player/Female/Medium/Bodies/F_MED_BudBlossom/Meshes/F_MED_BudBlossom_Physics
bodies {
 calf_r [{shapeType=Capsule, length=47.2139, radius=7.1537, shapeName=calf_r_capsule, center=<21.709 0 0>, rotation=<rot 90 0 0>}]
 dyn_skirt_01_bk [{shapeType=Capsule, length=34.505, radius=6.005, shapeName=dyn_skirt_01_bk_capsule, center=<8.1116 0 -6.1>, rotation=<rot 0 0 -90>}]
 dyn_skirt_base_bk [{shapeType=Capsule, length=0.505, radius=0.505, shapeName=dyn_skirt_base_bk_capsule, rotation=<rot 0 -90 0>}]
 dyn_whip_1 [{shapeType=Capsule, length=9, radius=3.5, shapeName=dyn_whip_1_capsule, center=<7.4747 0 -3.2>, rotation=<rot 0 0 90>}]
 dyn_whip_2 [{shapeType=Capsule, length=8, radius=6.5, shapeName=dyn_whip_2_capsule, center=<6.8953 0 -4.7>, rotation=<rot 0 0 88.4756>}]
 dyn_whip_base [{shapeType=Capsule, length=0.505, radius=0.505, shapeName=dyn_whip_base_capsule}]
 dyn_whip_habdle [{shapeType=Capsule, length=5, radius=3, shapeName=dyn_whip_habdle_capsule, center=<10.9091 0 -0.2618>, rotation=<rot 90 0 0>}]
 foot_r [{shapeType=Capsule, length=13, radius=5.505, shapeName=foot_r_capsule, center=<3.6782 -5.7012 0>, rotation=<rot 0 10.4376 90>}]
 pelvis [{shapeType=Capsule, length=11.2342, radius=10.7656, shapeName=pelvis_capsule}]
 root [{shapeType=Sphere, radius=150, shapeName=root_sphere, center=<0 0 -146.1322>}]
 thigh_l [{shapeType=Capsule, length=19.8693, radius=10.2657, shapeName=thigh_l_capsule, center=<-17.0723 0 0>, rotation=<rot 90 180 0>}]
 thigh_r [{shapeType=Capsule, length=20.3693, radius=9.7657, shapeName=thigh_r_capsule, center=<17.0723 0 0>, rotation=<rot 90 0 0>}]
}
constraints [
 {bone1Name=dyn_whip_1, bone2Name=dyn_whip_base, swing1LimitDegrees=5, swing2LimitDegrees=25, twistLimitDegrees=45, twistMotion=Locked}
 {bone1Name=dyn_whip_2, bone2Name=dyn_whip_1, swing1LimitDegrees=15, swing2LimitDegrees=15, twistLimitDegrees=45, twistMotion=Locked}
 {bone1Name=dyn_whip_habdle, bone2Name=dyn_whip_1, swing1LimitDegrees=15, swing2LimitDegrees=10, twistLimitDegrees=45, twistMotion=Locked}
 {
  bone1Name=dyn_whip_base
  bone2Name=pelvis
  swing1LimitDegrees=45
  swing1Motion=Free
  swing2LimitDegrees=45
  swing2Motion=Free
  twistLimitDegrees=45
  twistMotion=Free
 }
 {
  bone1Name=thigh_r
  bone2Name=pelvis
  swing1LimitDegrees=45
  swing1Motion=Free
  swing2LimitDegrees=45
  swing2Motion=Free
  twistLimitDegrees=45
  twistMotion=Free
 }
 {
  bone1Name=calf_r
  bone2Name=thigh_r
  swing1LimitDegrees=45
  swing1Motion=Free
  swing2LimitDegrees=45
  swing2Motion=Free
  twistLimitDegrees=45
  twistMotion=Free
 }
 {
  bone1Name=foot_r
  bone2Name=calf_r
  swing1LimitDegrees=45
  swing1Motion=Free
  swing2LimitDegrees=45
  swing2Motion=Free
  twistLimitDegrees=45
  twistMotion=Free
 }
 {
  bone1Name=pelvis
  bone2Name=root
  swing1LimitDegrees=45
  swing1Motion=Free
  swing2LimitDegrees=45
  swing2Motion=Free
  twistLimitDegrees=45
  twistMotion=Free
 }
 {
  bone1Name=dyn_skirt_base_bk
  bone2Name=pelvis
  swing1LimitDegrees=45
  swing1Motion=Free
  swing2LimitDegrees=45
  swing2Motion=Free
  twistLimitDegrees=45
  twistMotion=Free
 }
 {bone1Name=dyn_skirt_01_bk, bone2Name=dyn_skirt_base_bk, swing1LimitDegrees=5, swing2LimitDegrees=20, twistLimitDegrees=5}
 {
  bone1Name=thigh_l
  bone2Name=pelvis
  swing1LimitDegrees=45
  swing1Motion=Free
  swing2LimitDegrees=45
  swing2Motion=Free
  twistLimitDegrees=45
  twistMotion=Free
 }
]
