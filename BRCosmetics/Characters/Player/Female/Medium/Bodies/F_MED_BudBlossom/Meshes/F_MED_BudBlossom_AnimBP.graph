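# Post-process style AnimBP for the female BudBlossom cosmetic body, adding secondary motion to the attached parent's copied pose.
# A Control Rig feeds two RigidBody sims (body and choker physics assets) into a cached pose.
# A second cache layers angular and linear spring AnimDynamics on the upper-torso bones, and bPbBlendEnabled selects between caches.
# Each update sets gravity override from a parameter asset and a Lod1 flag from current LOD.
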AnimBlueprint /BRCosmetics/Characters/Player/Female/Medium/Bodies/F_MED_BudBlossom/Meshes/F_MED_BudBlossom_AnimBP
target_skeleton=&/F_MED_BudBlossom_Skeleton
graphs [
 AnimGraph {nodes [
   Output Pose(Result=Blend Poses by bool(
     BlendPose_0=Use cached pose 'BodyDynCache'()
     BlendPose_1=Use cached pose 'DynamicsCache'()
     bActiveValue=Get bPbBlendEnabled() -> (bPbBlendEnabled:bool)
    )) {layerGroup=None}
   n1 = Control Rig!(
    Source=Copy Pose From Mesh() {bUseAttachedParent=true}
    Alpha
    Rban:bool=Get bRigidBodyEnabled()
    Lod1:bool=Get Lod1() -> (Lod1:bool)
    DBNO:bool=Get bIsDBNO()
   ) -> (Pose) {controlRigAssetReference {blueprintRigClass=&/F_MED_BudBlossom_CtrlRig.F_MED_BudBlossom_CtrlRig_C}}
   n2 = RigidBody!(ComponentPose=Local To Component(LocalPose=@n1), Alpha, OverrideWorldGravity:Vector="0,0,0", ExternalForce) -> (Pose) {
    bClampLinearTranslationLimitToRefPose=true
    bOverrideWorldGravity=true
    overridePhysicsAsset=&/F_MED_BudBlossom_Physics
    simulationSpace=BaseBoneSpace
    baseBoneRef {boneName=pelvis}
    simSpaceSettings {dampingAlpha=0, maxAngularAcceleration=30, maxAngularVelocity=3, maxLinearAcceleration=1800, maxLinearVelocity=300, worldAlpha=1}
   }
   n3 = RigidBody(ComponentPose=@n2, OverrideWorldGravity="0,0,0") {
    bClampLinearTranslationLimitToRefPose=true
    bOverrideWorldGravity=true
    overridePhysicsAsset=&/F_MED_BudBlossom_Choker_Physics
    simulationSpace=BaseBoneSpace
    baseBoneRef {boneName=spine_05}
    simSpaceSettings {dampingAlpha=0, maxAngularAcceleration=30, maxAngularVelocity=5, maxLinearAcceleration=1000, maxLinearVelocity=150, worldAlpha=1}
   }
   Save cached pose 'DynamicsCache'(Pose=Component To Local(ComponentPose=@n3)) {cacheName=DynamicsCache}
   n4 = "Transform (Modify) Bone - Bone: dyn_upperTorso_l"(ComponentPose=Local To Component(LocalPose=Use cached pose 'DynamicsCache'()), Alpha=1, Translation="0,0,1", Rotation="0,0,0", Scale="1,1,1") {translationMode=BMM_Additive, translationSpace=BCS_BoneSpace, boneToModify {boneName=dyn_upperTorso_l}, translation {z=1}}
   n5 = "Transform (Modify) Bone - Bone: dyn_upperTorso_r"(ComponentPose=@n4, Alpha=1, Translation="0,0,1", Rotation="0,0,0", Scale="1,1,1") {translationMode=BMM_Additive, translationSpace=BCS_BoneSpace, boneToModify {boneName=dyn_upperTorso_r}, translation {z=1}}
   n6 = Get Sim Space Settings() -> (Sim Space Settings:Anim Phys Sim Space Settings Structure)
   n7 = Anim Dynamics - Bone: dyn_upperTorso_r!(ComponentPose=@n5, Alpha=1, AngularSpringConstant:float=5000, SimSpaceSettings:Anim Phys Sim Space Settings Structure=@n6) -> (Pose) {
    angularBiasOverride=1
    angularDampingOverride=0.9999
    angularSpringConstant=5000
    bAngularSpring=true
    bGravityOverrideInSimSpace=true
    bOverrideAngularBias=true
    bOverrideAngularDamping=true
    bUseGravityOverride=true
    linearDampingOverride=0.9999
    linearSpringConstant=100
    simulationSpace=RootRelative
    boundBone {boneName=dyn_upperTorso_r}
    componentAppliedLinearAccClamp=<2 2 2>
    gravityOverride {z=-980}
    physicsBodyDefinitions [{
      boundBone {boneName=dyn_upperTorso_r}
      boxExtents=<22 22 22>
      constraintSetup {angularLimitsMax=<0 1 5>, angularLimitsMin=<0 -5 -5>, angularTarget=<1 0 0>}
      localJointOffset=<2 0 0>
     }]
    relativeSpaceBone {boneName=spine_04}
    simSpaceSettings {maxAngularAcceleration=45, maxAngularVelocity=8, simSpaceAngularAlpha=0.25}
   }
   n8 = Anim Dynamics - Bone: dyn_upperTorso_l!(ComponentPose=@n7, Alpha=1, AngularSpringConstant:float=5000, SimSpaceSettings:Anim Phys Sim Space Settings Structure=@n6) -> (Pose) {
    angularBiasOverride=1
    angularDampingOverride=0.9999
    angularSpringConstant=5000
    bAngularSpring=true
    bGravityOverrideInSimSpace=true
    bOverrideAngularBias=true
    bOverrideAngularDamping=true
    bUseGravityOverride=true
    linearDampingOverride=0.9999
    linearSpringConstant=100
    simulationSpace=RootRelative
    boundBone {boneName=dyn_upperTorso_l}
    componentAppliedLinearAccClamp=<2 2 2>
    gravityOverride {z=-980}
    physicsBodyDefinitions [{
      boundBone {boneName=dyn_upperTorso_l}
      boxExtents=<20 20 20>
      constraintSetup {angularLimitsMax=<0 1 5>, angularLimitsMin=<0 -5 -5>, angularTarget=<1 0 0>}
      localJointOffset=<2 0 0>
     }]
    relativeSpaceBone {boneName=spine_04}
    simSpaceSettings {maxAngularAcceleration=45, maxAngularVelocity=8, simSpaceAngularAlpha=0.25}
   }
   n9 = Anim Dynamics - Bone: dyn_upperTorso_r!(ComponentPose=@n8, Alpha=1, GravityOverride:Vector="0,0,-200", LinearSpringConstant:float=180) -> (Pose) {
    angularBiasOverride=1
    angularDampingOverride=0.99
    angularSpringConstant=10000
    bGravityOverrideInSimSpace=true
    bLinearSpring=true
    bOverrideAngularBias=true
    bOverrideLinearDamping=true
    bUseGravityOverride=true
    linearDampingOverride=1
    linearSpringConstant=180
    simulationSpace=RootRelative
    boundBone {boneName=dyn_upperTorso_r}
    componentAppliedLinearAccClamp=<5 5 5>
    gravityOverride {z=-200}
    physicsBodyDefinitions [{
      boundBone {boneName=dyn_upperTorso_r}
      boxExtents=<80 80 80>
      constraintSetup {angularLimitsMax=<0 0 0>, angularLimitsMin=<0 0 0>, angularTarget=<1 0 0>, linearAxesMax=<0 0.5 1>, linearAxesMin=<0 -0.5 -0.5>}
      localJointOffset=<0.2 0 0>
     }]
    relativeSpaceBone {boneName=spine_04}
    simSpaceSettings {maxAngularAcceleration=50, maxAngularVelocity=10, simSpaceAngularAlpha=0.1}
   }
   n10 = Anim Dynamics - Bone: dyn_upperTorso_l!(ComponentPose=@n9, Alpha=1, GravityOverride:Vector="0,0,-200", LinearSpringConstant:float=180) -> (Pose) {
    angularBiasOverride=1
    angularDampingOverride=0.99
    angularSpringConstant=10000
    bGravityOverrideInSimSpace=true
    bLinearSpring=true
    bOverrideAngularBias=true
    bOverrideLinearDamping=true
    bUseGravityOverride=true
    linearDampingOverride=1
    linearSpringConstant=180
    simulationSpace=RootRelative
    boundBone {boneName=dyn_upperTorso_l}
    componentAppliedLinearAccClamp=<5 5 5>
    gravityOverride {z=-200}
    physicsBodyDefinitions [{
      boundBone {boneName=dyn_upperTorso_l}
      boxExtents=<90 90 90>
      constraintSetup {angularLimitsMax=<0 0 0>, angularLimitsMin=<0 0 0>, angularTarget=<1 0 0>, linearAxesMax=<0 0.5 1>, linearAxesMin=<0 -0.5 -0.5>}
      localJointOffset=<0.2 0 0>
     }]
    relativeSpaceBone {boneName=spine_04}
    simSpaceSettings {maxAngularAcceleration=50, maxAngularVelocity=10, simSpaceAngularAlpha=0.1}
   }
   Save cached pose 'BodyDynCache'(Pose=Component To Local(ComponentPose=@n10)) {cacheName=BodyDynCache}
  ]}
 EventGraph {nodes [
   Event BlueprintUpdateAnimation()
   > n1 = "Gravity Override State Function (Parameter Asset) (Nativized)"(
    CharacterPartAnimInstance=Self-Reference()
    GameTimeInSeconds=GetGameTimeInSeconds()
    Parameters=&/F_MED_BudBlossom_Parameters
   )
   > Set Gravity Override(Gravity Override=@n1)
   > Set Lod1(Lod1:bool="integer > integer"(A=Get CurrentLOD())) -> (Output_Get:bool)
  ]}
]
variables [
 Gravity Override {type=Vector}
 Lod1 {type=boolean}
 Sim Space Settings {
  type=AnimPhysSimSpaceSettings
  default {maxAngularAcceleration=45, maxAngularVelocity=8, simSpaceAngularAlpha=0.25, externalAngularVelocity=<0 0 0>}
 }
]
