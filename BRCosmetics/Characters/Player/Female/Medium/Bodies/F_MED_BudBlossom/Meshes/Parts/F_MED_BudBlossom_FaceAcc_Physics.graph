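# Physics asset for the BudBlossom female character's face accessory part, defining capsule bodies and joints for dynamic hair and bow bones.
# Head, glasses, neck, upper spine, clavicle and upper-arm capsules form the body structure, linked by angularly free constraints.
# Small base capsules join hair and bows to the head, each leading to angle-limited chains, including a branching right-side lock.
# Bow pieces act as hinges, swinging on one axis only.
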PhysicsAsset /BRCosmetics/Characters/Player/Female/Medium/Bodies/F_MED_BudBlossom/Meshes/Parts/F_MED_BudBlossom_FaceAcc_Physics
bodies {
 clavicle_l [{shapeType=Capsule, length=8.505, radius=3.005, shapeName=clavicle_l_capsule, center=<5.4339 0 0>, rotation=<rot 90 180 0>}]
 clavicle_r [
  {shapeType=Capsule, length=8.505, radius=3.005, shapeName=clavicle_r_capsule, center=<-5.4339 0 0>, rotation=<rot 90 0 0>}
  {shapeType=Capsule, length=7, radius=3, shapeName=clavicle_r_capsule, center=<-4.978 2.6846 -3.459>, rotation=<rot 90 0 0>}
 ]
 correction_glasses [{
   shapeType=Capsule
   length=9.9128
   radius=1.7186
   shapeName=correction_glasses_capsule
   center=<10.9931 -0.0059 -0.0748>
   rotation=<rot 0 -0.0334 -89.9959>
  }]
 dyn_bow_base_down_l [{shapeType=Capsule, length=0.505, radius=0.505, shapeName=dyn_bow_base_down_l_capsule, rotation=<rot 0 -90 0>}]
 dyn_bow_base_down_r [{shapeType=Capsule, length=0.505, radius=0.505, shapeName=dyn_bow_base_down_r_capsule, rotation=<rot 0 -90 0>}]
 dyn_bow_base_up_l [{shapeType=Capsule, length=0.505, radius=0.505, shapeName=dyn_bow_base_up_l_capsule, rotation=<rot 0 -90 0>}]
 dyn_bow_base_up_r [{shapeType=Capsule, length=0.505, radius=0.505, shapeName=dyn_bow_base_up_r_capsule, rotation=<rot 0 -90 0>}]
 dyn_bow_down_l [{shapeType=Capsule, length=1.7509, radius=1.4281, shapeName=dyn_bow_down_l_capsule, center=<1.9029 0 0>, rotation=<rot 0 0 90>}]
 dyn_bow_down_r [{shapeType=Capsule, length=1.6361, radius=1.2994, shapeName=dyn_bow_down_r_capsule, center=<1.4317 0 0>, rotation=<rot 0 0 90>}]
 dyn_bow_up_l [{shapeType=Capsule, length=1.7231, radius=1.3173, shapeName=dyn_bow_up_l_capsule, center=<1.6559 0 0>, rotation=<rot 0 0 90>}]
 dyn_bow_up_r [{shapeType=Capsule, length=1.6705, radius=1.4257, shapeName=dyn_bow_up_r_capsule, center=<1.4317 0 0>, rotation=<rot 0 0 90>}]
 dyn_hair_01_fr_l [{shapeType=Capsule, length=2.6122, radius=1.2904, shapeName=dyn_hair_01_fr_l_capsule, center=<2.7153 0 0>, rotation=<rot 90 0 0>}]
 dyn_hair_01_fr_r [{shapeType=Capsule, length=2.5117, radius=1.5507, shapeName=dyn_hair_01_fr_r_capsule, center=<1.9689 0 0>, rotation=<rot 90 0 0>}]
 dyn_hair_01_l [{shapeType=Capsule, length=3.8513, radius=1.4096, shapeName=dyn_hair_01_l_capsule, center=<3.1304 0 0>, rotation=<rot 90 0 0>}]
 dyn_hair_01_r [{shapeType=Capsule, length=2.1014, radius=1.583, shapeName=dyn_hair_01_r_capsule, center=<2.1198 0 0>, rotation=<rot 90 0 0>}]
 dyn_hair_1_01_side_l [{shapeType=Capsule, length=3.6563, radius=1.7774, shapeName=dyn_hair_1_01_side_l_capsule, center=<2.6181 0 0>, rotation=<rot 90 0 0>}]
 dyn_hair_1_01_side_r [{shapeType=Capsule, length=4.6157, radius=2.3281, shapeName=dyn_hair_1_01_side_r_capsule, center=<1.8534 0 0>, rotation=<rot 90 0 0>}]
 dyn_hair_1_02_side_l [{shapeType=Capsule, length=4.2503, radius=1.6736, shapeName=dyn_hair_1_02_side_l_capsule, center=<2.6181 0 0>, rotation=<rot 90 0 0>}]
 dyn_hair_1_02_side_r [{shapeType=Capsule, length=4.5234, radius=2, shapeName=dyn_hair_1_02_side_r_capsule, center=<1.8534 0 0>, rotation=<rot 90 0 0>}]
 dyn_hair_1_03_side_r [{shapeType=Capsule, length=4.46, radius=2, shapeName=dyn_hair_1_03_side_r_capsule, center=<1.8534 0 0>, rotation=<rot 90 0 0>}]
 dyn_hair_1_base_side_l [{shapeType=Capsule, length=0.505, radius=0.505, shapeName=dyn_hair_1_base_side_l_capsule}]
 dyn_hair_2_01_side_l [{shapeType=Capsule, length=3.2, radius=1.7483, shapeName=dyn_hair_2_01_side_l_capsule, center=<3 0 0>, rotation=<rot 90 0 0>}]
 dyn_hair_2_01_side_r [{shapeType=Capsule, length=3.7146, radius=1.8287, shapeName=dyn_hair_2_01_side_r_capsule, center=<3.4336 0 0>, rotation=<rot 90 0 0>}]
 dyn_hair_2_base_side_l [{shapeType=Capsule, length=0.505, radius=0.505, shapeName=dyn_hair_2_base_side_l_capsule, rotation=<rot 0 -90 0>}]
 dyn_hair_base_bk [{shapeType=Capsule, length=0.505, radius=0.505, shapeName=dyn_hair_base_bk_capsule}]
 dyn_hair_base_fr_l [{shapeType=Capsule, length=0.505, radius=0.505, shapeName=dyn_hair_base_fr_l_capsule, rotation=<rot 0 -90 0>}]
 dyn_hair_base_fr_r [{shapeType=Capsule, length=0.505, radius=0.505, shapeName=dyn_hair_base_fr_r_capsule, rotation=<rot 0 -90 0>}]
 dyn_hair_base_l [{shapeType=Capsule, length=0.505, radius=0.505, shapeName=dyn_hair_base_l_capsule, rotation=<rot 0 -90 0>}]
 dyn_hair_base_r [{shapeType=Capsule, length=0.505, radius=0.505, shapeName=dyn_hair_base_r_capsule, rotation=<rot 0 -90 0>}]
 dyn_hair_base_side_r [{shapeType=Capsule, length=0.505, radius=0.505, shapeName=dyn_hair_base_side_r_capsule}]
 dyn_hair_bk [{shapeType=Capsule, length=25.5528, radius=2.1652, shapeName=dyn_hair_bk_capsule, center=<1.9333 0 0>, rotation=<rot 90 0 0>}]
 dyn_hair_main_side_r [{shapeType=Capsule, length=2.5, radius=2, shapeName=dyn_hair_main_side_r_capsule, center=<1.4251 0 0>, rotation=<rot 90 0 0>}]
 head [{shapeType=Capsule, length=15.1142, radius=3.5, shapeName=head_capsule, rotation=<rot 0 0 90>}]
 neck_01 [{shapeType=Capsule, length=0.005, radius=4.005, shapeName=neck_01_capsule, center=<1.9514 0 0>, rotation=<rot 90 0 0>}]
 neck_02 [{shapeType=Capsule, length=0.5, radius=3.505, shapeName=neck_02_capsule, center=<2.2582 0 0>, rotation=<rot 90 0 0>}]
 spine_05 [
  {shapeType=Capsule, length=13, radius=7.005, shapeName=spine_05_capsule}
  {shapeType=Capsule, length=4, radius=4.5, shapeName=spine_05_capsule, center=<7.0343 0 0>, rotation=<rot 90 0 0>}
 ]
 upperarm_l [{shapeType=Capsule, length=23, radius=3.005, shapeName=upperarm_l_capsule, center=<12.3713 -0.944 0>, rotation=<rot 90 180 0>}]
 upperarm_r [{shapeType=Capsule, length=23, radius=3.005, shapeName=upperarm_r_capsule, center=<-12.3713 0.944 0>, rotation=<rot 90 0 0>}]
}
constraints [
 {bone1Name=dyn_hair_1_03_side_r, bone2Name=dyn_hair_1_02_side_r, swing1LimitDegrees=15, swing2LimitDegrees=10, twistLimitDegrees=5}
 {bone1Name=dyn_hair_1_02_side_r, bone2Name=dyn_hair_1_01_side_r, swing1LimitDegrees=20, swing2LimitDegrees=20, twistLimitDegrees=5}
 {bone1Name=dyn_hair_1_02_side_l, bone2Name=dyn_hair_1_01_side_l, swing1LimitDegrees=30, swing2LimitDegrees=20, twistLimitDegrees=5}
 {bone1Name=dyn_hair_1_01_side_l, bone2Name=dyn_hair_1_base_side_l, swing1LimitDegrees=25, swing2LimitDegrees=15, twistLimitDegrees=5}
 {bone1Name=dyn_hair_2_01_side_l, bone2Name=dyn_hair_2_base_side_l, swing1LimitDegrees=20, swing2LimitDegrees=15, twistLimitDegrees=5}
 {
  bone1Name=upperarm_r
  bone2Name=clavicle_r
  swing1LimitDegrees=45
  swing1Motion=Free
  swing2LimitDegrees=45
  swing2Motion=Free
  twistLimitDegrees=45
  twistMotion=Free
 }
 {
  bone1Name=upperarm_l
  bone2Name=clavicle_l
  swing1LimitDegrees=45
  swing1Motion=Free
  swing2LimitDegrees=45
  swing2Motion=Free
  twistLimitDegrees=45
  twistMotion=Free
 }
 {
  bone1Name=dyn_hair_1_base_side_l
  bone2Name=head
  swing1LimitDegrees=45
  swing1Motion=Free
  swing2LimitDegrees=45
  swing2Motion=Free
  twistLimitDegrees=45
  twistMotion=Free
 }
 {
  bone1Name=dyn_hair_2_base_side_l
  bone2Name=head
  swing1LimitDegrees=45
  swing1Motion=Free
  swing2LimitDegrees=45
  swing2Motion=Free
  twistLimitDegrees=45
  twistMotion=Free
 }
 {
  bone1Name=head
  bone2Name=neck_02
  swing1LimitDegrees=45
  swing1Motion=Free
  swing2LimitDegrees=45
  swing2Motion=Free
  twistLimitDegrees=45
  twistMotion=Free
 }
 {
  bone1Name=neck_02
  bone2Name=neck_01
  swing1LimitDegrees=45
  swing1Motion=Free
  swing2LimitDegrees=45
  swing2Motion=Free
  twistLimitDegrees=45
  twistMotion=Free
 }
 {
  bone1Name=neck_01
  bone2Name=spine_05
  swing1LimitDegrees=45
  swing1Motion=Free
  swing2LimitDegrees=45
  swing2Motion=Free
  twistLimitDegrees=45
  twistMotion=Free
 }
 {
  bone1Name=clavicle_l
  bone2Name=spine_05
  swing1LimitDegrees=45
  swing1Motion=Free
  swing2LimitDegrees=45
  swing2Motion=Free
  twistLimitDegrees=45
  twistMotion=Free
 }
 {
  bone1Name=clavicle_r
  bone2Name=spine_05
  swing1LimitDegrees=45
  swing1Motion=Free
  swing2LimitDegrees=45
  swing2Motion=Free
  twistLimitDegrees=45
  twistMotion=Free
 }
 {
  bone1Name=dyn_hair_base_bk
  bone2Name=head
  swing1LimitDegrees=45
  swing1Motion=Free
  swing2LimitDegrees=45
  swing2Motion=Free
  twistLimitDegrees=45
  twistMotion=Free
 }
 {bone1Name=dyn_hair_bk, bone2Name=dyn_hair_base_bk, swing1LimitDegrees=5, swing2LimitDegrees=15, twistLimitDegrees=0, twistMotion=Locked}
 {
  bone1Name=dyn_hair_base_fr_l
  bone2Name=head
  swing1LimitDegrees=45
  swing1Motion=Free
  swing2LimitDegrees=45
  swing2Motion=Free
  twistLimitDegrees=45
  twistMotion=Free
 }
 {
  bone1Name=dyn_hair_01_fr_l
  bone2Name=dyn_hair_base_fr_l
  swing1LimitDegrees=10
  swing2LimitDegrees=15
  twistLimitDegrees=45
  twistMotion=Locked
 }
 {
  bone1Name=dyn_hair_base_fr_r
  bone2Name=head
  swing1LimitDegrees=45
  swing1Motion=Free
  swing2LimitDegrees=45
  swing2Motion=Free
  twistLimitDegrees=45
  twistMotion=Free
 }
 {
  bone1Name=dyn_hair_01_fr_r
  bone2Name=dyn_hair_base_fr_r
  swing1LimitDegrees=10
  swing2LimitDegrees=15
  twistLimitDegrees=45
  twistMotion=Locked
 }
 {
  bone1Name=dyn_hair_base_l
  bone2Name=head
  swing1LimitDegrees=45
  swing1Motion=Free
  swing2LimitDegrees=45
  swing2Motion=Free
  twistLimitDegrees=45
  twistMotion=Free
 }
 {bone1Name=dyn_hair_01_l, bone2Name=dyn_hair_base_l, swing1LimitDegrees=12, swing2LimitDegrees=12, twistLimitDegrees=45, twistMotion=Locked}
 {
  bone1Name=dyn_hair_base_r
  bone2Name=head
  swing1LimitDegrees=45
  swing1Motion=Free
  swing2LimitDegrees=45
  swing2Motion=Free
  twistLimitDegrees=45
  twistMotion=Free
 }
 {bone1Name=dyn_hair_01_r, bone2Name=dyn_hair_base_r, swing1LimitDegrees=10, swing2LimitDegrees=20, twistLimitDegrees=45, twistMotion=Locked}
 {
  bone1Name=correction_glasses
  bone2Name=head
  swing1LimitDegrees=45
  swing1Motion=Free
  swing2LimitDegrees=45
  swing2Motion=Free
  twistLimitDegrees=45
  twistMotion=Free
 }
 {
  bone1Name=dyn_bow_base_up_l
  bone2Name=head
  swing1LimitDegrees=45
  swing1Motion=Free
  swing2LimitDegrees=45
  swing2Motion=Free
  twistLimitDegrees=45
  twistMotion=Free
 }
 {
  bone1Name=dyn_bow_up_l
  bone2Name=dyn_bow_base_up_l
  swing1LimitDegrees=45
  swing1Motion=Locked
  swing2LimitDegrees=15
  twistLimitDegrees=45
  twistMotion=Locked
 }
 {
  bone1Name=dyn_bow_down_l
  bone2Name=dyn_bow_base_down_l
  swing1LimitDegrees=45
  swing1Motion=Locked
  swing2LimitDegrees=15
  twistLimitDegrees=45
  twistMotion=Locked
 }
 {
  bone1Name=dyn_bow_base_down_l
  bone2Name=head
  swing1LimitDegrees=45
  swing1Motion=Free
  swing2LimitDegrees=45
  swing2Motion=Free
  twistLimitDegrees=45
  twistMotion=Free
 }
 {
  bone1Name=dyn_bow_base_down_r
  bone2Name=head
  swing1LimitDegrees=45
  swing1Motion=Free
  swing2LimitDegrees=45
  swing2Motion=Free
  twistLimitDegrees=45
  twistMotion=Free
 }
 {
  bone1Name=dyn_bow_down_r
  bone2Name=dyn_bow_base_down_r
  swing1LimitDegrees=45
  swing1Motion=Locked
  swing2LimitDegrees=20
  twistLimitDegrees=45
  twistMotion=Locked
 }
 {
  bone1Name=dyn_bow_base_up_r
  bone2Name=head
  swing1LimitDegrees=45
  swing1Motion=Free
  swing2LimitDegrees=45
  swing2Motion=Free
  twistLimitDegrees=45
  twistMotion=Free
 }
 {
  bone1Name=dyn_bow_up_r
  bone2Name=dyn_bow_base_up_r
  swing1LimitDegrees=45
  swing1Motion=Locked
  swing2LimitDegrees=15
  twistLimitDegrees=45
  twistMotion=Locked
 }
 {bone1Name=dyn_hair_2_01_side_r, bone2Name=dyn_hair_main_side_r, swing1LimitDegrees=15, swing2LimitDegrees=10, twistLimitDegrees=5}
 {bone1Name=dyn_hair_1_01_side_r, bone2Name=dyn_hair_main_side_r, swing1LimitDegrees=15, swing2LimitDegrees=10, twistLimitDegrees=5}
 {bone1Name=dyn_hair_main_side_r, bone2Name=dyn_hair_base_side_r, swing1LimitDegrees=25, swing2LimitDegrees=10, twistLimitDegrees=5}
 {
  bone1Name=dyn_hair_base_side_r
  bone2Name=head
  swing1LimitDegrees=45
  swing1Motion=Free
  swing2LimitDegrees=45
  swing2Motion=Free
  twistLimitDegrees=45
  twistMotion=Free
 }
]
