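# Animation Blueprint for the BudBlossom female medium character's face accessory, copying its pose from the attached parent mesh.
# A Control Rig receives the rigid-body enabled flag, then feeds a RigidBody node simulating in head-bone space with clamped motion.
# Each update stores a gravity override computed from the character's parameter asset, though the RigidBody gravity input shows a fixed downward vector rather than that variable.
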AnimBlueprint /BRCosmetics/Characters/Player/Female/Medium/Bodies/F_MED_BudBlossom/Meshes/Parts/F_MED_BudBlossom_FaceAcc_AnimBP
target_skeleton=&/F_MED_BudBlossom_FaceAcc_Skeleton
graphs [
 AnimGraph {nodes [
   n1 = Control Rig!(Source=Copy Pose From Mesh() {bUseAttachedParent=true}, Alpha, RBAN:bool=Get bRigidBodyEnabled()) -> (Pose) {controlRigAssetReference {blueprintRigClass=&/F_MED_BudBlossom_FaceAcc_CtrlRig.F_MED_BudBlossom_FaceAcc_CtrlRig_C}}
   n2 = RigidBody!(ComponentPose=Local To Component(LocalPose=@n1), Alpha, OverrideWorldGravity:Vector="0,0,-980", ExternalForce) -> (Pose) {
    bClampLinearTranslationLimitToRefPose=true
    bOverrideWorldGravity=true
    overridePhysicsAsset=&/F_MED_BudBlossom_FaceAcc_Physics
    simulationSpace=BaseBoneSpace
    baseBoneRef {boneName=head}
    overrideWorldGravity {z=-980}
    simSpaceSettings {dampingAlpha=0, maxAngularAcceleration=40, maxAngularVelocity=8, maxLinearAcceleration=2000, maxLinearVelocity=500, worldAlpha=1}
   }
   Output Pose(Result=Component To Local(ComponentPose=@n2)) {layerGroup=None}
  ]}
 EventGraph {nodes [
   Event BlueprintUpdateAnimation()
   > n1 = "Gravity Override State Function (Parameter Asset) (Nativized)"(
    CharacterPartAnimInstance=Self-Reference()
    GameTimeInSeconds=GetGameTimeInSeconds()
    Parameters=/BRCosmetics/Characters/Player/Female/Medium/Bodies/F_MED_BudBlossom/Meshes/F_MED_BudBlossom_Parameters
   )
   > Set Gravity Override(Gravity Override=@n1)
  ]}
]
variables [Gravity Override {type=Vector}]
